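# Additive unlit Niagara mesh-particle material that fakes glowing candle subsurface scattering for the TracePaper player emote.
# Three weighted sines at golden-ratio-related frequencies, offset per particle, flicker between deeper red and a brighter blend of UV1 gradient and emissive texture, tinted by particle color.
# With depth testing disabled, a pixel-versus-scene depth tolerance mask multiplies particle alpha and a UV1 falloff.
# A small offset vector toward the camera is also computed.
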Material /Game/Animation/Game/MainPlayer/Emotes/TracePaper/FX/Materials/M_Emote_TracePaper_CandleSSS
graph {
 properties {
  bDisableDepthTest=true
  bUsedWithNiagaraMeshParticles=true
  blendMode=BLEND_Additive
  pixelDepthOffsetMode=PDOM_Legacy
  shadingModel=MSM_Unlit
  usedShadingModels=Unlit
 }
 nodes [
  n1 = VectorParameter() {parameterName=Deeper Red, defaultValue=<rgba 0.3 0 0 1>}
  n2 = VectorParameter() {parameterName=Deep Red, defaultValue=<rgba 2 0.1 0 0.1>}
  n3 = ComponentMask(None=TextureCoordinate() {coordinateIndex=1}) {g=true}
  n4 = Multiply(A=@n2.RGB, B=Power(Base=@n3, Exp=@n2.A))
  n5 = TextureSample() {texture=/Game/Animation/Props/Gadgets_TracePaper/Textures/T_Emote_Prop_TracePaper_E}
  n6 = Multiply(A=Add(A=Time(), B=ParticleRandom()), B=ScalarParameter() {defaultValue=1, parameterName=Flicker Speed})
  n7 = Sine(None=@n6)
  n8 = Sine(None=Multiply(A=@n6) {constB=1.618})
  n9 = Sine(None=Multiply(A=@n6) {constB=2.618})
  n10 = VectorParameter() {parameterName=Flicker Weights, defaultValue=<rgba 1 0.8 0.5 0>}
  n11 = Multiply(
   A=MaterialFunctionCall(R=ConstantBiasScale(None=@n7), G=ConstantBiasScale(None=@n8), B=ConstantBiasScale(None=@n9)) {materialFunction=&AppendMany}.RGB
   B=@n10.RGB
  )
  n12 = Multiply(
   A=Divide(
    A=MaterialFunctionCall(f3=@n11) {materialFunction=&AddComponents}.f3
    B=MaterialFunctionCall(f3=@n10.RGB) {materialFunction=&AddComponents}.f3
   )
   B=@n1.A
  )
  n13 = ParticleColor()
  Multiply(A=LinearInterpolate(A=@n1.RGB, B=Max(A=@n4, B=@n5.RGB), Alpha=OneMinus(None=@n12)), B=@n13.RGB)
  n14 = CameraPositionWS()
  n15 = WorldPosition()
  Multiply(A=Normalize(VectorInput=Subtract(A=@n14, B=@n15.XYZ)), B=ScalarParameter() {defaultValue=0.1, parameterName=Camera Offset})
  n16 = OneMinus(None=@n3)
  n17 = ScalarParameter() {defaultValue=8, parameterName=Distance Fade Exponent}
  n18 = PixelDepth()
  n19 = SceneDepth()
  n20 = Subtract(A=Abs(None=Subtract(A=@n18.R, B=@n19.R)), B=ScalarParameter() {defaultValue=0.5, parameterName=Depth Tolerance})
  Multiply(A=Multiply(A=@n13.A, B=Power(Base=@n16, Exp=@n17)), B=OneMinus(None=Saturate(None=@n20)))
  ParticleRadius()
 ]
}
